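# Niagara cosmetic effect for the male Virtual Shadow player body, giving off bright green-to-white particles and faint grey smoke.
# Minimal pins four spawn groups to skeletal mesh bones or sockets with offsets; Minimal001 spawns two from position offsets.
# Hand and head hidden user bools zero matching spawn rates.
# Spawn group drives color, motion and VisibilityTag for mesh and sprite renderers, animated through scale curves, SubUV and dynamic material parameters.
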NiagaraSystem /BRCosmetics/Characters/Player/Male/Medium/Bodies/M_MED_Virtual_Shadow/FX/NS_VirtualShadow_Smoke
dependencies {
 usedDataInterfaces [
  &NiagaraDataInterfaceArrayPosition
  &NiagaraDataInterfaceCurve
  &NiagaraDataInterfaceSpriteRendererInfo
  &NiagaraDataInterfaceArrayColor
  &NiagaraDataInterfaceSkeletalMesh
 ]
 usedDynamicInputs []
 usedModules [
  &self:SystemSpawnScript.SystemScriptSource.SystemScriptGraph.NiagaraNodeAssignment_0.SetVariables_#1
  &SystemState
 ]
}
emitters {
 Minimal {
  emitterUpdateScript [
   &EmitterState()
   &SpawnRate(
    SpawnRate=&MakeCustomFloatFromBool(Bool=$User.bisHandHidden, True Float=0, False Float=5)
    Use Spawn Probability=false
   )
   &SpawnRate(
    SpawnGroup=1
    SpawnRate=&MakeCustomFloatFromBool(Bool=$User.bisHandHidden, True Float=0, False Float=5)
    Use Spawn Probability=false
   ) {module=SpawnRate001}
   &SpawnRate(
    SpawnGroup=2
    SpawnRate=&MakeCustomFloatFromBool(Bool=$User.bisHandHidden, True Float=0, False Float=&RandomRangeFloat(Minimum=3, Maximum=6))
    Use Spawn Probability=false
   ) {module=SpawnRate002}
   &SpawnRate(
    SpawnGroup=3
    SpawnRate=&MakeCustomFloatFromBool(Bool=$User.bisHandHidden, True Float=0, False Float=&RandomRangeFloat(Minimum=3, Maximum=6))
    Use Spawn Probability=false
   ) {module=SpawnRate003}
  ]
  particleSpawnScript [
   /SetVariables(Particles.VisibilityTag=&MakeCustomIntFromBool(
     Bool=&SetBoolByIntComparison(A=&SpawnGroup(), B=2, Comparison Type=::NewEnumerator2)
     True INT=1
     False INT=0
    )) {isSetParameters=true, module=SetVariables_#2}
   &InitializeParticle(
    Color=&MaskLinearColorBySpawnGroup(
     Masked Linear Color=&RandomRangeLinearColor(Minimum=<rgba 0 0.5 0 20>, Maximum=<rgba 1 1 1 20>)
     Passthrough Linear Color=&MaskLinearColorBySpawnGroup(
      Masked Linear Color=&RandomRangeLinearColor(Minimum=<rgba 0 0.5 0 20>, Maximum=<rgba 1 1 1 20>)
      Passthrough Linear Color=&RandomRangeLinearColor(Minimum=<rgba 0.2 0.2 0.2 0.15>, Maximum=<rgba 0.5 0.5 0.5 0.4>)
     )
    )
    Lifetime=&MakeCustomFloatFromBool(
     Bool=&SetBoolByIntComparison(A=&SpawnGroup(), B=2, Comparison Type=::NewEnumerator2)
     True Float=&RandomRangeFloat(Minimum=1.6, Maximum=2.5)
     False Float=&RandomRangeFloat(Minimum=1, Maximum=1.5)
    )
    Mesh Scale Max=<0.03 0.003 2>
    Mesh Scale Min=<0.05 0.005 1>
    Mesh Scale Mode=::NewEnumerator2
    Sprite Rotation Mode=::NewEnumerator2
    Sprite Size Mode=::NewEnumerator3
    Uniform Sprite Size=&RandomRangeFloat(Minimum=20, Maximum=25)
    UsePositionOffset=false
    Write Material Random=true
   )
   &AddVelocity(Velocity=&MaskVectorBySpawnGroup(
     Masked Vector=<1 0 -1>
     Passthrough Vector=&MaskVectorBySpawnGroup(
      Masked Vector=<0 0 1>
      Passthrough Vector=&RandomRangeVector(Minimum=<6 -5 3>, Maximum=<16 5 5>)
      Spawn Group Mask=1
     )
    ))
   &CameraOffset(Camera Offset Amount=&MaskFloatBySpawnGroup(
     Masked Float=&RandomRangeFloat(Minimum=5, Maximum=8)
     Passthrough Float=&MaskFloatBySpawnGroup(Masked Float=&RandomRangeFloat(Minimum=5, Maximum=8), Passthrough Float=2, Spawn Group Mask=3)
     Spawn Group Mask=2
    ))
   /SetVariables(Particles.Sprite_DynamicParam:Vector4f=&MakeVector4(
     X=&RandomRangeFloat(Minimum=0, Maximum=1)
     Y=&RandomRangeFloat(Minimum=1, Maximum=2)
     Z=0
     W=0.5
    )) {isSetParameters=true, module=SetVariables_#3}
  ]
  particleUpdateScript [
   &ParticleState()
   &SkeletalMeshLocation(
    Bone / Socket Index=&SpawnGroup()
    Bone Sampling Mode=::NewEnumerator4
    MaskSpawnGroup=false
    Sampled Position Offset=/Niagara/DynamicInputs/Arrays/SelectPositionFromArray(
     Array Sampling Mode:ENiagaraArraySamplingMode=::NewEnumerator1
     Direct Array Index:NiagaraInt32=&SpawnGroup()
     Interpolation:NiagaraFloat=?
     Direct Array Mode:ENiagara_WrapClamp=::NewEnumerator0
     Evaluation Type:ENiagaraRandomnessEvaluation=::NewEnumerator0
     Normalized Interpolation:NiagaraBool=true
     Recalculate Random Each Loop:NiagaraBool=false
     Position Array:NiagaraDataInterfaceArrayPosition=DataInterface() {propertyValues {GpuSyncMode=SyncCpuToGpu, MaxElements=0, PositionData [<10 0 0>, <-7 0 0>, <8 0 0>, <-8 0 0>]}}
     Randomness Mode:ENiagaraRandomnessMode=::NewEnumerator0
     Fixed Random Seed:NiagaraBool=false
     Random Seed:NiagaraInt32=?
    )
    Skeletal Mesh=$System.Skeletal Mesh
   )
   &Drag(
    Drag=&MakeCustomFloatFromBool(
     Bool=&SetBoolByIntComparison(A=&SpawnGroup(), B=2, Comparison Type=::NewEnumerator2)
     True Float=1
     False Float=0
    )
    Use Linear Drag=true
    Use Rotational Drag=false
   )
   &SolveForcesAndVelocity(Clamp Velocity=false, Limit Acceleration=false)
   &ScaleSpriteSize(Uniform Curve Sprite Scale=DataInterface() {propertyValues {Curve {keys [
        {arriveTangent=7.2451, interpMode=RCIM_Cubic, leaveTangent=7.2451, tangentMode=RCTM_User, time=0, value=0.2}
        {arriveTangent=1.0307, interpMode=RCIM_Cubic, leaveTangent=1.0307, tangentMode=RCTM_User, time=0.1, value=0.7}
        <at 0.95 1 RCIM_Cubic>
       ]}}})
   &SpriteRotationRate(Rotation Rate=&RandomRangeFloat(Minimum=-60, Maximum=-170))
   &ScaleMeshSize(Scale Factor=&MakeVector(
     X=&FloatFromCurve(FloatCurve=DataInterface() {propertyValues {Curve {keys [
          {arriveTangent=12.4313, interpMode=RCIM_Cubic, leaveTangent=12.4313, tangentMode=RCTM_User, time=0, value=0.1}
          {interpMode=RCIM_Cubic, leaveTangent=6.4408, time=1, value=10}
         ]}}})
     Y=&FloatFromCurve(FloatCurve=DataInterface() {propertyValues {Curve {keys [{arriveTangent=8.2274, interpMode=RCIM_Cubic, leaveTangent=8.2274, tangentMode=RCTM_User, time=0, value=1}, <at 1 8 RCIM_Cubic>]}}})
     Z=1
    ))
   &DynamicMaterialParameters(
    Index 0 Param 1=&FloatFromCurve(FloatCurve=DataInterface() {propertyValues {Curve {keys [<at 0 5>,<at 1 0>]}}})
    Index 0 Param 2=&RandomRangeFloat(Minimum=20, Maximum=50)
    Index 0 Param 3=&FloatFromCurve(FloatCurve=DataInterface() {propertyValues {Curve {keys [<at 0 3>,<at 1 12>]}}})
    Index 0 Param 4=&FloatFromCurve(FloatCurve=DataInterface() {propertyValues {Curve {keys [
         <at 0.01 0.25 RCIM_Cubic>
         {arriveTangent=-0.1102, interpMode=RCIM_Cubic, leaveTangent=-0.1102, tangentMode=RCTM_User, time=0.82, value=0.22}
         {arriveTangent=-0.9272, interpMode=RCIM_Cubic, leaveTangent=-0.9272, tangentMode=RCTM_User, time=0.9, value=0.1}
         {arriveTangent=-1.2311, interpMode=RCIM_Cubic, leaveTangent=-1.2311, tangentMode=RCTM_User, time=1, value=0}
        ]}}})
    Module.Index0Float0Write=true
    Module.Index0Float1Write=true
    Module.Index0Float2Write=true
    Module.Index0Float3Write=true
   )
   &ScaleColor(
    Scale Alpha=&MakeCustomFloatFromBool(
     Bool=&SetBoolByIntComparison(A=&SpawnGroup(), B=2, Comparison Type=::NewEnumerator2)
     True Float=&FloatFromCurve(FloatCurve=DataInterface() {propertyValues {Curve {keys [
          {arriveTangent=3.5511, interpMode=RCIM_Cubic, time=0, value=0}
          {arriveTangent=2.9969, interpMode=RCIM_Cubic, leaveTangent=2.9969, tangentMode=RCTM_User, time=0.25, value=1}
          {arriveTangentWeight=0.1843, interpMode=RCIM_Cubic, leaveTangentWeight=0.067, tangentMode=RCTM_User, time=0.8, value=1.8}
          <at 1 0 RCIM_Cubic>
         ]}}})
     False Float=&FloatFromCurve(FloatCurve=DataInterface() {propertyValues {Curve {keys [
          <at 0 0 RCIM_Cubic>
          {arriveTangent=3, interpMode=RCIM_Cubic, leaveTangent=3, time=0.25, value=1}
          {arriveTangent=-0.7143, interpMode=RCIM_Cubic, leaveTangent=-0.7143, time=0.4, value=1.2}
          {arriveTangent=-2.7143, interpMode=RCIM_Cubic, leaveTangent=-2.7143, time=0.6, value=0.75}
          {arriveTangent=-1.4178, interpMode=RCIM_Cubic, leaveTangent=-1.4178, tangentMode=RCTM_User, time=0.75, value=0.25}
          {arriveTangent=-0.8887, interpMode=RCIM_Cubic, leaveTangent=-0.8887, tangentMode=RCTM_User, time=0.9, value=0.09}
          <at 1 0 RCIM_Cubic>
         ]}}})
    )
    Scale RGB=&MakeCustomVectorFromBool(
     Bool=&SetBoolByIntComparison(A=&SpawnGroup(), B=2, Comparison Type=::NewEnumerator2)
     True Vector=&VectorFromFloat(Value=&FloatFromCurve(FloatCurve=DataInterface() {propertyValues {Curve {keys [<at 0 1>,<at 0.5 0.2>]}}}))
     False Vector=&VectorFromFloat(Value=&FloatFromCurve(FloatCurve=DataInterface() {propertyValues {Curve {keys [<at 0 3>,<at 0.3 1>]}}}))
    )
    ScaleA=true
    ScaleRGB=true
   )
   &SubUVAnimation(
    Direct Frame Index:NiagaraFloat=&FloatFromCurve(FloatCurve=DataInterface() {propertyValues {Curve {keys [
         {arriveTangent=26.2429, interpMode=RCIM_Cubic, leaveTangent=26.2429, tangentMode=RCTM_User, time=0, value=25}
         {arriveTangent=10.3514, interpMode=RCIM_Cubic, leaveTangent=10.3514, tangentMode=RCTM_User, time=1, value=55}
        ]}}})
    "Sprite Renderer "=DataInterface() {propertyValues {SpriteRenderer=&self:Minimal.NiagaraSpriteRendererProperties_1}}
    SubUV Animation Mode:ENiagara_SubUVLookupModeV2=::NewEnumerator5
   )
  ]
  rendererClasses [&NiagaraMeshRendererProperties, &NiagaraSpriteRendererProperties]
  renderers [
   {rendererClass=&NiagaraMeshRendererProperties, rendererIndex=0}
   {rendererClass=&NiagaraSpriteRendererProperties, rendererIndex=1}
  ]
 }
 Minimal001 {
  emitterUpdateScript [
   &EmitterState()
   &SpawnRate(
    SpawnRate=&MakeCustomFloatFromBool(Bool=$User.bisHeadHidden, True Float=0, False Float=2)
    Use Spawn Probability=false
   )
   &SpawnRate(
    SpawnGroup=1
    SpawnRate=&MakeCustomFloatFromBool(Bool=$User.bisHandHidden, True Float=0, False Float=&RandomRangeFloat(Minimum=3, Maximum=6))
    Use Spawn Probability=false
   ) {module=SpawnRate001}
  ]
  particleSpawnScript [
   /SetVariables(Particles.VisibilityTag=&SpawnGroup()) {isSetParameters=true, module=SetVariables_#2}
   &InitializeParticle(
    Color Maximum=&SelectLinearColorFromArray(
     Array Sampling Mode=::NewEnumerator1
     Direct Array Index=&SpawnGroup()
     Color Selection Array=DataInterface() {propertyValues {ColorData [<rgba 1 1 1 20>, <rgba 0.4947 0.4947 0.4947 0.3>]}}
     Random Seed=?
    )
    Color Minimum=&SelectLinearColorFromArray(
     Array Sampling Mode=::NewEnumerator1
     Direct Array Index=&SpawnGroup()
     Color Selection Array=DataInterface() {propertyValues {ColorData [<rgba 0 0.5 0 1>, <rgba 0.2 0.2 0.2 0.1>]}}
     Random Seed=?
    )
    Color Mode=::NewEnumerator2
    Lifetime=&MaskFloatBySpawnGroup(
     Masked Float=&RandomRangeFloat(Minimum=1, Maximum=1.5)
     Passthrough Float=&RandomRangeFloat(Minimum=1.5, Maximum=2.5)
    )
    Mesh Scale Max=<0.04 0.003 2>
    Mesh Scale Min=<0.06 0.008 1>
    Mesh Scale Mode=::NewEnumerator2
    Position Offset=&MaskVectorBySpawnGroup(
     Masked Vector=&RandomRangeVector(Minimum=<0 -15 0>, Maximum=<-10 -10 0>)
     Passthrough Vector=&RandomRangeVector(Minimum=<-9 4 0>, Maximum=<-13 3 0>)
    )
    Sprite Rotation Mode=::NewEnumerator2
    Sprite Size Mode=::NewEnumerator3
    Uniform Sprite Size=&RandomRangeFloat(Minimum=20, Maximum=25)
    UsePositionOffset=true
    Write Material Random=true
   )
   &AddVelocity(Velocity=&MaskVectorBySpawnGroup(
     Masked Vector=&RandomRangeVector(Minimum=<0 8 5>, Maximum=<0 4 -5>)
     Passthrough Vector=&RandomRangeVector(Minimum=<6 -5 3>, Maximum=<16 5 5>)
    ))
   &CameraOffset(Camera Offset Amount=&MaskFloatBySpawnGroup(
     Masked Float=&RandomRangeFloat(Minimum=5, Maximum=8)
     Passthrough Float=&MaskFloatBySpawnGroup(Masked Float=&RandomRangeFloat(Minimum=5, Maximum=8), Passthrough Float=2, Spawn Group Mask=3)
     Spawn Group Mask=2
    ))
   /SetVariables(Particles.Sprite_DynamicParam:Vector4f=&MakeVector4(
     X=&RandomRangeFloat(Minimum=0, Maximum=1)
     Y=&RandomRangeFloat(Minimum=1, Maximum=2)
     Z=0
     W=0.5
    )) {isSetParameters=true, module=SetVariables_#3}
  ]
  particleUpdateScript [
   &ParticleState()
   &Drag(
    Drag=&MakeCustomFloatFromBool(
     Bool=&SetBoolByIntComparison(A=&SpawnGroup(), B=2, Comparison Type=::NewEnumerator2)
     True Float=1
     False Float=0
    )
    Use Linear Drag=true
    Use Rotational Drag=false
   )
   &SolveForcesAndVelocity(Clamp Velocity=false, Limit Acceleration=false)
   &ScaleSpriteSize(Uniform Curve Sprite Scale=DataInterface() {propertyValues {Curve {keys [
        {arriveTangent=7.2451, interpMode=RCIM_Cubic, leaveTangent=7.2451, tangentMode=RCTM_User, time=0, value=0.2}
        {arriveTangent=1.0307, interpMode=RCIM_Cubic, leaveTangent=1.0307, tangentMode=RCTM_User, time=0.1, value=0.7}
        <at 0.95 1 RCIM_Cubic>
       ]}}})
   &SpriteRotationRate(Rotation Rate=&RandomRangeFloat(Minimum=-60, Maximum=-170))
   &ScaleMeshSize(Scale Factor=&MakeVector(
     X=&FloatFromCurve(FloatCurve=DataInterface() {propertyValues {Curve {keys [
          {arriveTangent=12.4313, interpMode=RCIM_Cubic, leaveTangent=12.4313, tangentMode=RCTM_User, time=0, value=0.1}
          {interpMode=RCIM_Cubic, leaveTangent=6.4408, time=1, value=10}
         ]}}})
     Y=&FloatFromCurve(FloatCurve=DataInterface() {propertyValues {Curve {keys [{arriveTangent=8.2274, interpMode=RCIM_Cubic, leaveTangent=8.2274, tangentMode=RCTM_User, time=0, value=1}, <at 1 8 RCIM_Cubic>]}}})
     Z=1
    ))
   &DynamicMaterialParameters(
    Index 0 Param 1=&FloatFromCurve(FloatCurve=DataInterface() {propertyValues {Curve {keys [<at 0 5>,<at 1 0>]}}})
    Index 0 Param 2=&RandomRangeFloat(Minimum=20, Maximum=50)
    Index 0 Param 3=&FloatFromCurve(FloatCurve=DataInterface() {propertyValues {Curve {keys [<at 0 3>,<at 1 12>]}}})
    Index 0 Param 4=&FloatFromCurve(FloatCurve=DataInterface() {propertyValues {Curve {keys [
         <at 0.01 0.25 RCIM_Cubic>
         {arriveTangent=-0.1102, interpMode=RCIM_Cubic, leaveTangent=-0.1102, tangentMode=RCTM_User, time=0.82, value=0.22}
         {arriveTangent=-0.9272, interpMode=RCIM_Cubic, leaveTangent=-0.9272, tangentMode=RCTM_User, time=0.9, value=0.1}
         {arriveTangent=-1.2311, interpMode=RCIM_Cubic, leaveTangent=-1.2311, tangentMode=RCTM_User, time=1, value=0}
        ]}}})
    Module.Index0Float0Write=true
    Module.Index0Float1Write=true
    Module.Index0Float2Write=true
    Module.Index0Float3Write=true
   )
   &ScaleColor(
    Scale Alpha=&MakeCustomFloatFromBool(
     Bool=&SetBoolByIntComparison(A=&SpawnGroup(), B=2, Comparison Type=::NewEnumerator2)
     True Float=&FloatFromCurve(FloatCurve=DataInterface() {propertyValues {Curve {keys [
          {arriveTangent=3.5511, interpMode=RCIM_Cubic, time=0, value=0}
          {arriveTangent=2.9969, interpMode=RCIM_Cubic, leaveTangent=2.9969, tangentMode=RCTM_User, time=0.25, value=1}
          {arriveTangentWeight=0.1843, interpMode=RCIM_Cubic, leaveTangentWeight=0.067, tangentMode=RCTM_User, time=0.8, value=1.8}
          <at 1 0 RCIM_Cubic>
         ]}}})
     False Float=&FloatFromCurve(FloatCurve=DataInterface() {propertyValues {Curve {keys [
          <at 0 0 RCIM_Cubic>
          {arriveTangent=3, interpMode=RCIM_Cubic, leaveTangent=3, time=0.25, value=1}
          {arriveTangent=-0.7143, interpMode=RCIM_Cubic, leaveTangent=-0.7143, time=0.4, value=1.2}
          {arriveTangent=-2.7143, interpMode=RCIM_Cubic, leaveTangent=-2.7143, time=0.6, value=0.75}
          {arriveTangent=-1.4178, interpMode=RCIM_Cubic, leaveTangent=-1.4178, tangentMode=RCTM_User, time=0.75, value=0.25}
          {arriveTangent=-0.8887, interpMode=RCIM_Cubic, leaveTangent=-0.8887, tangentMode=RCTM_User, time=0.9, value=0.09}
          <at 1 0 RCIM_Cubic>
         ]}}})
    )
    Scale RGB=&MakeCustomVectorFromBool(
     Bool=&SetBoolByIntComparison(A=&SpawnGroup(), B=1, Comparison Type=::NewEnumerator2)
     True Vector=&VectorFromFloat(Value=&FloatFromCurve(FloatCurve=DataInterface() {propertyValues {Curve {keys [<at 0 1>,<at 0.5 0.2>]}}}))
     False Vector=&MakeVector(
      X=1
      Y=&FloatFromCurve(FloatCurve=DataInterface() {propertyValues {Curve {keys [<at 0 3>,<at 0.3 1>]}}})
      Z=1
     )
    )
    ScaleA=true
    ScaleRGB=true
   )
   &SubUVAnimation(
    Direct Frame Index:NiagaraFloat=&FloatFromCurve(FloatCurve=DataInterface() {propertyValues {Curve {keys [
         {arriveTangent=26.2429, interpMode=RCIM_Cubic, leaveTangent=26.2429, tangentMode=RCTM_User, time=0, value=25}
         {arriveTangent=10.3514, interpMode=RCIM_Cubic, leaveTangent=10.3514, tangentMode=RCTM_User, time=1, value=55}
        ]}}})
    "Sprite Renderer "=DataInterface() {propertyValues {SpriteRenderer=&self:Minimal001_1.NiagaraSpriteRendererProperties_1}}
    SubUV Animation Mode:ENiagara_SubUVLookupModeV2=::NewEnumerator5
   )
  ]
  rendererClasses [&NiagaraMeshRendererProperties, &NiagaraSpriteRendererProperties]
  renderers [
   {rendererClass=&NiagaraMeshRendererProperties, rendererIndex=0}
   {rendererClass=&NiagaraSpriteRendererProperties, rendererIndex=1}
  ]
 }
}
summary {
 systemName=NS_VirtualShadow_Smoke
 emitters [
  {emitterName=Minimal, rendererClasses [&NiagaraMeshRendererProperties, &NiagaraSpriteRendererProperties]}
  {emitterName=Minimal001, rendererClasses [&NiagaraMeshRendererProperties, &NiagaraSpriteRendererProperties]}
 ]
 userVariables [
  User.bisHandHidden {type=NiagaraBool}
  User.bisHeadHidden {type=NiagaraBool}
 ]
}
