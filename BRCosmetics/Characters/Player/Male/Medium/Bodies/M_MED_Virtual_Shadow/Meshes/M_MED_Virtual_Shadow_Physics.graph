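# Physics asset for the M_MED_Virtual_Shadow male body, covering its dangling belt pieces, sword and left leg pouch.
# Capsules on pelvis, thighs and right calf form the body base, with boxes for sword and pouch.
# Belt pieces chain to the right thigh, the sword hanging off the side belt; the pouch chains to the left thigh twist bone, carrying a flap and four shells.
# MSTR anchor bones attach with free angular motion, while their dynamic children have angular limits.
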PhysicsAsset /BRCosmetics/Characters/Player/Male/Medium/Bodies/M_MED_Virtual_Shadow/Meshes/M_MED_Virtual_Shadow_Physics
bodies {
 calf_r [{shapeType=Capsule, length=25.0172, radius=7.6293, center=<19.6148 3.1911 1.3758>, rotation=<rot 87.2174 0.0002 0.0002>}]
 dyn_belt_back [{shapeType=Capsule, length=5.983, radius=3.2158, center=<0.2182 0 0>, rotation=<rot 90 0 0>}]
 dyn_belt_back_MSTR [{shapeType=Capsule, length=1.307, radius=0.7275}]
 dyn_belt_front [{shapeType=Capsule, length=5.9948, radius=4.2055, center=<3.5676 0.2751 -0.2918>, rotation=<rot 90 0 0>}]
 dyn_belt_front_MSTR [{shapeType=Capsule, length=0, radius=0.6619}]
 dyn_belt_side [{shapeType=Capsule, length=6.6363, radius=3.7776, center=<0.8971 0.0001 0>, rotation=<rot 0 0 -90>}]
 dyn_belt_sword [{shapeType=Box, extentX=5.9996, extentY=15.8694, extentZ=54.0991, center=<4.9518 4.2628 -5.856>, rotation=<rot 22.9488 -1.5026 -26.7497>}]
 dyn_leg_pouch_flap [{shapeType=Capsule, length=1.3763, radius=0.505, center=<1.7747 -0.1449 -0.0086>, rotation=<rot 0 -90.4921 -89.3994>}]
 dyn_leg_pouch_l [{shapeType=Box, extentX=19.8388, extentY=17.08, extentZ=6.3382, center=<0 0.382 -3.4845>, rotation=<rot -4.1971 0 0>}]
 dyn_leg_pouch_l_MSTR [{shapeType=Capsule, length=0.505, radius=0.505, rotation=<rot 0 -90 0>}]
 dyn_leg_pouch_shell01 [{shapeType=Capsule, length=3.4221, radius=1.507, center=<0.075 -0.1778 0.0819>, rotation=<rot 0 -91.6673 -88.8235>}]
 dyn_leg_pouch_shell02 [{shapeType=Capsule, length=3.4591, radius=1.6743, center=<0.073 -0.1777 -0.0128>, rotation=<rot 0 -93.974 -89.1208>}]
 dyn_leg_pouch_shell03 [{shapeType=Capsule, length=3.4424, radius=1.7424, center=<0.1202 -0.1102 0.006>, rotation=<rot 0 -93.5161 -88.1066>}]
 dyn_leg_pouch_shell04 [{shapeType=Capsule, length=3.4597, radius=2.0233, center=<0.0642 -0.1775 -0.1927>, rotation=<rot 0 -94.3297 -87.3004>}]
 pelvis [{shapeType=Capsule, length=9.0965, radius=9.0346, center=<0 -3.6451 0.0002>}]
 thigh_l [{shapeType=Capsule, length=22.2473, radius=6.6935, center=<-17.3275 0.2446 -0.5142>, rotation=<rot 86.7428 0 0>}]
 thigh_r [{shapeType=Capsule, length=25.3718, radius=7.552, center=<22.3435 -0.3141 -0.4441>, rotation=<rot 89.1649 179.9999 179.9999>}]
 thigh_twist_01_l [{shapeType=Capsule, length=1.216, radius=2.46, center=<8.9029 0.4878 -3.4382>, rotation=<rot 90 0 0>}]
}
constraints [
 {bone1Name=thigh_r, bone2Name=pelvis, swing1LimitDegrees=45, swing2LimitDegrees=45, twistLimitDegrees=45}
 {bone1Name=dyn_belt_sword, bone2Name=dyn_belt_side, swing1LimitDegrees=45, swing2LimitDegrees=45, twistLimitDegrees=30}
 {bone1Name=dyn_belt_side, bone2Name=thigh_r, swing1LimitDegrees=15, swing2LimitDegrees=15, twistLimitDegrees=15}
 {
  bone1Name=thigh_l
  bone2Name=pelvis
  swing1LimitDegrees=45
  swing1Motion=Free
  swing2LimitDegrees=45
  swing2Motion=Free
  twistLimitDegrees=45
  twistMotion=Free
 }
 {
  bone1Name=thigh_twist_01_l
  bone2Name=thigh_l
  swing1LimitDegrees=45
  swing1Motion=Free
  swing2LimitDegrees=45
  swing2Motion=Free
  twistLimitDegrees=45
  twistMotion=Free
 }
 {bone1Name=dyn_leg_pouch_l, bone2Name=dyn_leg_pouch_l_MSTR, swing1LimitDegrees=45, swing2LimitDegrees=45, twistLimitDegrees=15}
 {
  bone1Name=calf_r
  bone2Name=thigh_r
  swing1LimitDegrees=45
  swing1Motion=Free
  swing2LimitDegrees=45
  swing2Motion=Free
  twistLimitDegrees=45
  twistMotion=Free
 }
 {
  bone1Name=dyn_leg_pouch_l_MSTR
  bone2Name=thigh_twist_01_l
  swing1LimitDegrees=45
  swing1Motion=Free
  swing2LimitDegrees=45
  swing2Motion=Free
  twistLimitDegrees=45
  twistMotion=Free
 }
 {bone1Name=dyn_leg_pouch_flap, bone2Name=dyn_leg_pouch_l, swing1LimitDegrees=45, swing2LimitDegrees=45, twistLimitDegrees=5}
 {bone1Name=dyn_leg_pouch_shell01, bone2Name=dyn_leg_pouch_l, swing1LimitDegrees=20, swing2LimitDegrees=20, twistLimitDegrees=5}
 {bone1Name=dyn_leg_pouch_shell02, bone2Name=dyn_leg_pouch_l, swing1LimitDegrees=20, swing2LimitDegrees=20, twistLimitDegrees=5}
 {bone1Name=dyn_leg_pouch_shell03, bone2Name=dyn_leg_pouch_l, swing1LimitDegrees=20, swing2LimitDegrees=20, twistLimitDegrees=5}
 {bone1Name=dyn_leg_pouch_shell04, bone2Name=dyn_leg_pouch_l, swing1LimitDegrees=20, swing2LimitDegrees=20, twistLimitDegrees=5}
 {bone1Name=dyn_belt_back, bone2Name=dyn_belt_back_MSTR, swing1LimitDegrees=45, swing2LimitDegrees=45, twistLimitDegrees=15}
 {
  bone1Name=dyn_belt_back_MSTR
  bone2Name=thigh_r
  swing1LimitDegrees=45
  swing1Motion=Free
  swing2LimitDegrees=45
  swing2Motion=Free
  twistLimitDegrees=45
  twistMotion=Free
 }
 {
  bone1Name=dyn_belt_front_MSTR
  bone2Name=thigh_r
  swing1LimitDegrees=45
  swing1Motion=Free
  swing2LimitDegrees=45
  swing2Motion=Free
  twistLimitDegrees=45
  twistMotion=Free
 }
 {bone1Name=dyn_belt_front, bone2Name=dyn_belt_front_MSTR, swing1LimitDegrees=45, swing2LimitDegrees=45, twistLimitDegrees=15}
]
