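# Physics asset for the face accessory mesh of the Virtual Shadow male medium player body, covering its hood, horns and FX bones.
# A head capsule is the hub, with each dyn_ bone constrained directly to it in a flat layout rather than chained.
# The horns use boxes and the other bones use capsules.
# The main and side hood constraints lock swing and twist, while hood top, horns and FX bones stay angularly unlocked.
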PhysicsAsset /BRCosmetics/Characters/Player/Male/Medium/Bodies/M_MED_Virtual_Shadow/Meshes/Parts/M_MED_Virtual_Shadow_FaceAcc_Physics
bodies {
 dyn_fx_l [{shapeType=Capsule, length=2, radius=1, center=<0 0 -0.1>}]
 dyn_fx_r [{shapeType=Capsule, length=2, radius=1, center=<0 0 0.1>}]
 dyn_hood [{shapeType=Capsule, length=7.427, radius=1.8437}]
 dyn_hood_l [{shapeType=Capsule, length=3.7416, radius=1.7633, center=<0.6852 0 -0.0001>}]
 dyn_hood_r [{shapeType=Capsule, length=2.5846, radius=1.7, center=<0.9621 0 0.0001>}]
 dyn_hood_top [{shapeType=Capsule, length=4.0391, radius=2.4446, center=<1 0 0>}]
 dyn_horn_l [{shapeType=Box, extentX=12.1909, extentY=6.999, extentZ=1.5708, center=<0.1569 0.2528 0>}]
 dyn_horn_r [{shapeType=Box, extentX=11.9565, extentY=6.8972, extentZ=1.7243, center=<0.032 0.1683 0.2433>}]
 head [{shapeType=Capsule, length=5.3795, radius=10.3576}]
}
constraints [
 {bone1Name=dyn_hood_top, bone2Name=head, swing1LimitDegrees=45, swing2LimitDegrees=45, twistLimitDegrees=15}
 {
  bone1Name=dyn_hood
  bone2Name=head
  swing1LimitDegrees=45
  swing1Motion=Locked
  swing2LimitDegrees=45
  swing2Motion=Locked
  twistLimitDegrees=15
  twistMotion=Locked
 }
 {
  bone1Name=dyn_hood_l
  bone2Name=head
  swing1LimitDegrees=45
  swing1Motion=Locked
  swing2LimitDegrees=45
  swing2Motion=Locked
  twistLimitDegrees=15
  twistMotion=Locked
 }
 {
  bone1Name=dyn_hood_r
  bone2Name=head
  swing1LimitDegrees=45
  swing1Motion=Locked
  swing2LimitDegrees=45
  swing2Motion=Locked
  twistLimitDegrees=15
  twistMotion=Locked
 }
 {bone1Name=dyn_horn_l, bone2Name=head, swing1LimitDegrees=45, swing2LimitDegrees=45, twistLimitDegrees=15}
 {bone1Name=dyn_horn_r, bone2Name=head, swing1LimitDegrees=45, swing2LimitDegrees=45, twistLimitDegrees=15}
 {bone1Name=dyn_fx_l, bone2Name=head, swing1LimitDegrees=45, swing2LimitDegrees=45, twistLimitDegrees=15}
 {bone1Name=dyn_fx_r, bone2Name=head, swing1LimitDegrees=45, swing2LimitDegrees=45, twistLimitDegrees=15}
]
